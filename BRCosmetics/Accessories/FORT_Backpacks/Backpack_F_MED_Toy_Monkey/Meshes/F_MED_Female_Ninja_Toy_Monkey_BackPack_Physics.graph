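# Physics asset for the female Ninja Toy Monkey backpack cosmetic, defining simple collision and joint limits for its dynamic bones.
# Five capsule bodies cover the pack bone plus two two-link chains, dyn_top and dyn_bot.
# Each chain's first bone is constrained to the pack and its second bone to the first.
# Narrow swing limits and nearly locked twist restrict the chains to small swaying motion.
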PhysicsAsset /BRCosmetics/Accessories/FORT_Backpacks/Backpack_F_MED_Toy_Monkey/Meshes/F_MED_Female_Ninja_Toy_Monkey_BackPack_Physics
bodies {
 dyn_bot_0 [{shapeType=Capsule, length=4.6852, radius=3.8388, center=<6.288 0 0>, rotation=<rot 90 0 0>}]
 dyn_bot_1 [{shapeType=Capsule, length=2.755, radius=2.255, center=<4.0933 0 0>, rotation=<rot 90 0 0>}]
 dyn_top_0 [{shapeType=Capsule, length=5.6738, radius=4.1075, center=<-2.8835 -0.0002 0.0001>, rotation=<rot 90 0 0>}]
 dyn_top_1 [{shapeType=Capsule, length=2.7534, radius=2.2806, center=<3.2015 0 0.082>, rotation=<rot -88.5329 0 0>}]
 pack [{shapeType=Capsule, length=1.5929, radius=2.501, center=<2.9223 -2.0972 1.4702>, rotation=<rot -63.2943 0 62.4733>}]
}
constraints [
 {bone1Name=dyn_bot_0, bone2Name=pack, swing1LimitDegrees=5, swing2LimitDegrees=30, twistLimitDegrees=0.1}
 {bone1Name=dyn_bot_1, bone2Name=dyn_bot_0, swing1LimitDegrees=10, swing2LimitDegrees=30, twistLimitDegrees=0.1}
 {bone1Name=dyn_top_0, bone2Name=pack, swing1LimitDegrees=5, swing2LimitDegrees=10, twistLimitDegrees=1}
 {bone1Name=dyn_top_1, bone2Name=dyn_top_0, swing1LimitDegrees=20, swing2LimitDegrees=20, twistLimitDegrees=1}
]
