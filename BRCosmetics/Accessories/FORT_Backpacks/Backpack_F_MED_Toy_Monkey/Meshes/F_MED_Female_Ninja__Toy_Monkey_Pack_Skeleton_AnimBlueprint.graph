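# Animation Blueprint for the toy monkey backpack cosmetic, adding secondary motion and posture offsets.
# The output chain runs a RigidBody simulation from the reference pose, then two additive offsets on the pack bone.
# The event graph drives them with roll mapped from spine_05 yaw relative to spine_01, and an alpha switched on during skydiving or floating.
# Crouch/sprint root offsets, targeting blend and copied-mesh pose cache do not feed the output.
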AnimBlueprint /BRCosmetics/Accessories/FORT_Backpacks/Backpack_F_MED_Toy_Monkey/Meshes/F_MED_Female_Ninja__Toy_Monkey_Pack_Skeleton_AnimBlueprint
target_skeleton=&/F_MED_Female_Ninja__Toy_Monkey_Pack_Skeleton
graphs [
 AnimGraph {nodes [
   n1 = RigidBody(ComponentPose=Local To Component(LocalPose=Local Space Ref Pose()), Alpha=0.1) {
    bOverrideWorldGravity=true
    overridePhysicsAsset=&/F_MED_Female_Ninja_Toy_Monkey_BackPack_Physics
    simulationSpace=BaseBoneSpace
    alphaBoolBlend {blendOption=Linear}
    baseBoneRef {boneName=root}
    componentAppliedLinearAccClamp=<500 500 500>
    overrideWorldGravity {z=-980}
   }
   Blend Poses by bool(BlendPose_0=Component To Local(ComponentPose=@n1), BlendPose_1=Component To Local(), BlendTime_1=0.05, bActiveValue=Get bIsTargeting()) {blendType=Linear}
   Component To Local()
   "Transform (Modify) Bone"(Alpha=1, Translation="0,0,0", Rotation="0,0,0", Scale="1,1,1") {alphaBoolBlend {blendOption=Linear}, boneToModify=~}
   "Transform (Modify) Bone - Bone: dyn_bot_0"(Alpha=1, Translation="0,0,0", Rotation="0,0,0", Scale="1,1,1") {rotationMode=BMM_Additive, alphaBoolBlend {blendOption=Linear}, boneToModify {boneName=dyn_bot_0}}
   "Transform (Modify) Bone - Bone: root"!(
    ComponentPose
    Alpha
    Translation_X:double=0
    Translation_Y:double=Get backpack_moving() -> (backpack_moving:double)
    Translation_Z:double=0
    Rotation=Get backpack_rotation() -> (backpack_rotation:Rotator)
    Scale
   ) -> (Pose) {
    rotationMode=BMM_Additive
    rotationSpace=BCS_BoneSpace
    translationMode=BMM_Additive
    translationSpace=BCS_BoneSpace
    alphaBoolBlend {blendOption=Linear}
    boneToModify {boneName=root}
   }
   n2 = RigidBody(ComponentPose=Local To Component(LocalPose=Local Space Ref Pose()), Alpha=0.5) {
    bOverrideWorldGravity=true
    overridePhysicsAsset=&F_MED_Female_Ninja_Toy_Monkey_BackPack_Physics
    simulationSpace=BaseBoneSpace
    alphaBoolBlend {blendOption=Linear}
    baseBoneRef {boneName=root}
    componentAppliedLinearAccClamp=<500 500 500>
    overrideWorldGravity {z=-980}
   }
   n3 = "Transform (Modify) Bone - Bone: pack"(ComponentPose=@n2, Alpha=0.5, Translation="0,-3,0", Rotation=Get backpack_rotate()) {
    rotationMode=BMM_Additive
    rotationSpace=BCS_BoneSpace
    translationMode=BMM_Additive
    translationSpace=BCS_BoneSpace
    alphaBoolBlend {blendOption=Linear}
    boneToModify {boneName=pack}
    translation {y=-3}
   }
   n4 = "Transform (Modify) Bone - Bone: pack"(ComponentPose=@n3, Alpha=Get backpack_translate() -> (backpack_translate:double), Translation="3,2,0", Rotation="0,-12,-13") {
    rotationMode=BMM_Additive
    rotationSpace=BCS_BoneSpace
    translationMode=BMM_Additive
    translationSpace=BCS_BoneSpace
    alphaBoolBlend {blendOption=Linear}
    boneToModify {boneName=pack}
    rotation {roll=-13, yaw=-12}
    translation=<3 2>
   }
   Output Pose(Result=Component To Local(ComponentPose=@n4)) {layerGroup=None}
   Save cached pose 'AnimCache'(Pose=Copy Pose From Mesh() {bCopyCurves=true, bUseAttachedParent=true}) {cacheName=AnimCache}
  ]}
 EventGraph {nodes [
   Event BlueprintUpdateAnimation()
   > n1 = Sequence()
   > Set skydivingRotation(skydivingRotation:Rotator=SelectRotator!(A:Rotator="30,0,0", B:Rotator="0, 0, 0", bPickA:bool=Get bIsSkyDiving()) -> (ReturnValue:Rotator)) -> (Output_Get:Rotator)
   n2 = GetOwnerSkeletalMeshComponent()
   n3 = BreakRotator(InRot=BreakTransform(InTransform=GetDeltaTransformFromRefPose(self=@n2, BoneName=spine_05, BaseName=spine_01)).Rotation)
   Set backpack_rotate(execute=@n1.then_1, backpack_rotate=MakeRotator(Roll=MapRangeClamped(Value=@n3.Yaw, InRangeA=12, InRangeB=-30, OutRangeA=8, OutRangeB=30)))
   n4 = Get bIsSkydiveFloating()
   n5 = Get bIsSkyDiving()
   > Set backpack_translate(backpack_translate:double="To Float (Boolean)"(InBool=OR Boolean(A=@n4, B=@n5))) -> (Output_Get:double)
   n6 = Get bIsCrouching()
   n7 = AND Boolean(A="float > float"(A=Get PawnSpeed2D(), B=560), B=Get bIsOnGround() -> (bIsOnGround:bool))
   Set backpack_rotation(backpack_rotation:Rotator=MakeRotator(Roll=SelectFloat(A=-10, B=0, bPickA=@n6), Pitch=SelectFloat(A=10, B=0, bPickA=@n7), Yaw=SelectFloat(A=-15, B=0, bPickA=@n7))) -> (Output_Get:Rotator)
   > Set backpack_moving(backpack_moving:double=SelectFloat(A=-1, B=0, bPickA=@n7)) -> (Output_Get:double)
  ]}
]
variables [
 backpack_moving {type=number}
 backpack_rotate {type=Rotator}
 backpack_rotation {type=Rotator}
 backpack_translate {type=number}
 skydivingRotation {type=Rotator}
]
